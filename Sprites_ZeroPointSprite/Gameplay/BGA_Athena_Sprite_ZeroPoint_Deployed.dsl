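; Deployed Zero Point sprite building actor that projects an AOE shield when its ability executes.
; The server spawns and tracks the shield, while clients play the montage, burst FX and looping audio.
; A timed delay switches the montage's current loop section to its matching outro as the ability ends.
; Sleep or EndPlay removes the shield and fades audio; BeginPlay and waking register tag density with the Living World manager.
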(event Custom|Awoken)

(event EventExecuteSpriteAbility
  (bind _soundlibrary (Variables|Default|GetSoundLibrary))
  (bind _activatemontage (Variables|Default|GetActivateMontage))
  ;[1 1 1] Spawn AOE BGA on Ability Execution
  (Networking|SwitchHasAuthority
    (:Authority
      (bind _returnvalue (Game|SpawnBuildingActorfromClass "/Sprites_ZeroPointSprite/Gameplay/Deployed/BGA_Athena_Sprite_ZeroPointShield.BGA_Athena_Sprite_ZeroPointShield_C" (Math|Transform|MakeTransform (Transformation|GetActorLocation)) 0 false 0 (Game|GetInstigator)))
      (Variables|Default|SetSpawnedShield _returnvalue))
    (:Remote
      ;[1 1 1] Delay so that we can outro the activation montage as the ability is ending
      (Networking|CanExecuteCosmeticEvents
        (:True
          (Utilities|FlowControl|Sequence
            (:then_0
              (Animation|Montage|PlayMontage (Variables|Default|GetSkeletalMeshComponentBudgeted) _activatemontage)
              (bind _returnvalue_2 (ContextualAnim|Utilities|GetSectionLength _activatemontage (- (Montage|GetNumSections _activatemontage) 1)))
              (Utilities|FlowControl|Delay (Math|Float|Max(Float) (- (ScalableFloat|GetValueatLevel (Variables|Default|GetHF_AISpawnTimer)) (+ _returnvalue_2 (ScalableFloat|GetValueatLevel (Variables|HFVars|GetRow_ActivationAnimationEndBuffer))))))
              ;[1 1 1] Outro the activation animation
              (bind _assprite_zero_point_creature_skeleton_anim_blueprint (Utilities|Casting|CastToSprite_ZeroPoint_Creature_Skeleton_AnimBlueprint (Components|SkeletalMesh|GetAnimInstance (Variables|Default|GetSkeletalMeshComponentBudgeted))))
              (Utilities|FlowControl|Sequence
                (:then_0
                  (if (== (Animation|Montage|MontageGetCurrentSection _assprite_zero_point_creature_skeleton_anim_blueprint 0) "Loop1")
                    (Animation|Montage|MontageSetNextSection _assprite_zero_point_creature_skeleton_anim_blueprint (Animation|Montage|MontageGetCurrentSection _assprite_zero_point_creature_skeleton_anim_blueprint 0) "Outro1" 0)))
                (:then_1
                  (if (== (Animation|Montage|MontageGetCurrentSection _assprite_zero_point_creature_skeleton_anim_blueprint 0) "Loop2")
                    (Animation|Montage|MontageSetNextSection _assprite_zero_point_creature_skeleton_anim_blueprint (Animation|Montage|MontageGetCurrentSection _assprite_zero_point_creature_skeleton_anim_blueprint 0) "Outro2" 0)))
                (:then_2
                  (if (== (Animation|Montage|MontageGetCurrentSection _assprite_zero_point_creature_skeleton_anim_blueprint 0) "Loop3")
                    (Animation|Montage|MontageSetNextSection _assprite_zero_point_creature_skeleton_anim_blueprint (Animation|Montage|MontageGetCurrentSection _assprite_zero_point_creature_skeleton_anim_blueprint 0) "Outro3" 0)))))
            (:then_1
              (Niagara|SpawnSystemAttached "/Sprites_ZeroPointSprite/Assets/FX/NS_SuperSprite_AoE_DeployBurst.NS_SuperSprite_AoE_DeployBurst" (Variables|Default|GetStaticMesh) "None" "0, 0, 0" "0, 0, 0" "KeepRelativeOffset" true)
              (SoundLibrary|PlayEvent _soundlibrary "(TagName=\"SoundLibrary.Sprite.Deploy.Start\")")
              (bind _returnvalue_3 (SoundLibrary|PlayEvent _soundlibrary "(TagName=\"SoundLibrary.Sprite.Deploy.Loop\")"))
              (if (Utilities|Array|IsValidIndex (.AudioComponents (Utilities|Struct|BreakSoundLibraryPlayResult _returnvalue_3)))
                (Variables|Default|SetAudioLoop (Utilities|Array|Get(acopy) (.AudioComponents (Utilities|Struct|BreakSoundLibraryPlayResult _returnvalue_3))))))))))))

(event Custom|RegisterAwokenSprite
  (bind _tags_density (Variables|Dep%28Deletelater%29|GetTags%20Density))
  ;[0.9971 1 0.9978 0.5] Play Activate Montage if applicable, otherwise go to Sleep.
  (Networking|SwitchHasAuthority
    (:Authority
      (if (not (Utilities|Array|IsEmpty _tags_density))
        (Class|FortAthenaLivingWorldManager|RegisterTagDensityReservoir (Class|FortAthenaLivingWorldManager|GetLivingWorldManager) self _tags_density)))))

(event EventBeginPlay
  (bind _tags_density (Variables|Dep%28Deletelater%29|GetTags%20Density))
  ;[1 1 1] Begin Play
  (|CallParentFunction)
  ;[0.9971 1 0.9978 0.5] Play Activate Montage if applicable, otherwise go to Sleep.
  (if (Variables|Default|GetIsSleeping)
    (CallFunction|EnterSleepState))
  (Networking|SwitchHasAuthority
    (:Authority
      (if (not (Utilities|Array|IsEmpty _tags_density))
        (Class|FortAthenaLivingWorldManager|RegisterTagDensityReservoir (Class|FortAthenaLivingWorldManager|GetLivingWorldManager) self _tags_density)))))

(event EventWakeUpSprite
  ;[1 1 1 0.5] On Wake Up
  (|CallParentFunction)
  (CallFunction|RegisterAwokenSprite)
  (CallFunction|Awoken))

(event EventEnterSleepState
  ;[1 1 1 0.5] On Sleep
  (|CallParentFunction)
  (Utilities|FlowControl|Sequence
    (:then_0
      (bind _assprite_zero_point_creature_skeleton_anim_blueprint (Utilities|Casting|CastToSprite_ZeroPoint_Creature_Skeleton_AnimBlueprint (Components|SkeletalMesh|GetAnimInstance (Variables|Default|GetSkeletalMeshComponentBudgeted))))
      (Variables|/Script/SpritesRuntime.SpriteCreatureAnimInstance|SetbIsCooldown true _assprite_zero_point_creature_skeleton_anim_blueprint)
      (Animation|Montage|MontageStop _assprite_zero_point_creature_skeleton_anim_blueprint 0.1 0))
    (:then_1
      (CallFunction|RemoveShield))
    (:then_2
      (Components|Activation|Deactivate (Variables|Default|GetNS_SuperSprite_Deployed)))))

(event EventMulticast_HandleBinding
  (params (Actor :type Actor%20Object%20Reference) (HitComp :type Primitive%20Component%20Object%20Reference))
  ;[1 1 1] Handle Binding
  (|CallParentFunction Actor HitComp))

(event Custom|SetLocationFromServer
  (params (ServerLocation :type Vector))
  ;[1 1 1] Set Location From Server
  (Transformation|SetActorLocation self ServerLocation false true)
  (Variables|Default|SetLocation ServerLocation))

(event EventEndPlay
  (params (EndPlayReason :type EEndPlayReason%20Enum :default "Destroyed"))
  (|CallParentFunction EndPlayReason)
  (CallFunction|RemoveShield))

(event EventOnDeathServer
  (params (Damage :type Float%20%28single-precision%29) (DamageTags :type Gameplay%20Tag%20Container%20Structure%20%28by%20ref%29) (Momentum :type Vector :default "0, 0, 0") (HitInfo :type Hit%20Result%20Structure%20%28by%20ref%29) (InstigatedBy :type Controller%20Object%20Reference) (DamageCauser :type Actor%20Object%20Reference) (EffectContext :type Gameplay%20Effect%20Context%20Handle%20Structure))
  ;[1 1 1] On Death
  (|CallParentFunction Damage DamageTags Momentum HitInfo InstigatedBy DamageCauser EffectContext)
  (Rendering|SetActorHiddenInGame self true))

(event Custom|AddShieldAudio)

(event EventOutOfCharges
  (bind (_returnvalue_location _returnvalue_rotation) (Math|Transform|BreakTransform (Transformation|GetActorTransform)))
  (bind _returnvalue (Niagara|SpawnSystematLocation "/Sprites/FX/Niagara/NS_Sprite_Despawn.NS_Sprite_Despawn" _returnvalue_location _returnvalue_rotation "1.000000,1.000000,1.000000" false true "None" false))
  (Effects|Components|ParticleSystem|SetColorParameter _returnvalue "SpriteColor" "(R=2.500000,G=1.692359,B=0.000000,A=1.000000)"))

(event Custom|RemoveShield
  (Networking|SwitchHasAuthority
    (:Authority
      (bind _spawnedshield (Variables|Default|GetSpawnedShield)
        (:then
          (Class|BGAAthenaSpriteZeroPointShield|EndShield _spawnedshield)
          (Variables|Default|SetSpawnedShield 0))
        (:else)))
    (:Remote
      (SoundLibrary|PlayEvent (Variables|Default|GetSoundLibrary) "(TagName=\"SoundLibrary.Sprite.Deploy.Stop\")")
      (bind _audioloop (Variables|Default|GetAudioLoop)
        (:then
          (Audio|Components|Audio|FadeOut _audioloop (Variables|Default|GetLoopFadeTime) 0.0 "Linear" :!class /Script/Engine.AudioComponent))
        (:else)))))

(fn GetDynamicDestructionResourceType
  (return "Metal"))

(fn GetDynamicDestructionStaticMeshComponents
  (return (Utilities|Array|MakeArray (Variables|Default|GetStaticMesh)) 0 0))

(fn ShouldUseDynamicDestructionMaterial
  (return true))

(fn ShouldUseDynamicDestructionNiagaraSystem
  (return true))

(fn ConstructionScript
  (|CallParentFunction))
